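# Unlit translucent Niagara sprite material for the Marshmellow glider FX, drawing a soft radial glow or dark gradient.
# Zero-tiled UVs pan over time through a rainbow gradient, cycling one particle-tinted color, which UseBlackColor blends toward a vector color before HDR remapping.
# A centered sphere mask shapes the alpha, with dynamic parameters driving falloff and panning speed.
# Particle alpha multiplies in depth and near-camera fades, which mobile skips.
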Material /BRCosmetics/Gadgets/Assets/VinderTech_GliderChute/Glider_Marshmellow/FX/M_Glider_Mello_DarkGradient
graph {
 properties {bUsedWithNiagaraSprites=true, blendMode=BLEND_Translucent, pixelDepthOffsetMode=PDOM_Legacy, shadingModel=MSM_Unlit, usedShadingModels=Unlit}
 nodes [
  n1 = DynamicParameter() {defaultValue=<rgba 3 1 1 1>, paramNames [Falloff, Panning Speed, Param3, Param4]}
  n2 = Panner(Coordinate=ComponentMask(None=TextureCoordinate() {uTiling=0, vTiling=0}) {g=true}, Time=Multiply(A=Time(), B=@n1.Panning Speed)) {speedX=1.6667}
  n3 = ParticleColor()
  n4 = LinearInterpolate(
   A=Multiply(A=TextureSample(UVs=@n2) {texture=/Game/Global/Textures/Default/Masks_and_Gradients/T_Grad_Rainbow}.RGB, B=@n3.RGB)
   B=VectorParameter().RGB
   Alpha=ScalarParameter() {parameterName=UseBlackColor}
  )
  MaterialFunctionCall(Emissive=MaterialFunctionCall(LDR Input=Clamp(None=@n4), HDR Tint=@n4) {materialFunction=&DeriveHDRfromLDR}.RemappedRange) {materialFunction=&MF_FX_EmissiveAdjustments}
  n5 = TextureCoordinate()
  n6 = Constant() {r=0.5}
  n7 = DepthFade(
   Opacity=MaterialFunctionCall(Distance=ScalarParameter() {defaultValue=250, parameterName=NearCameraFadeDist}) {materialFunction=&Near_Camera_Fade}.Result
   FadeDistance=ScalarParameter() {defaultValue=200, parameterName=DepthFade}
  )
  n8 = Constant() {r=1}
  Multiply(
   A=Power(Base=SphereMask(A=@n5, B=@n6, Radius=@n6) {hardnessPercent=0}, Exp=@n1.Falloff) {constExponent=3}
   B=Multiply(A=@n3.A, B=ShadingPathSwitch(Default=@n7, Deferred=@n7, Mobile=@n8) {desc=Migrated from FeatureLevelSwitch})
  )
 ]
}
