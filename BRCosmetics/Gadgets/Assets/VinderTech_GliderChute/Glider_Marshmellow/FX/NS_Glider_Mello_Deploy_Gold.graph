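# Niagara effect for the deployment of the gold Marshmellow glider, built from four independent emitters.
# Center spawns short-lived gold flash sprites, while Mello and Mesh_Sprites burst a sprite and a gold-tinted mesh that pop up in scale from an overbright start.
# Curves drive their dynamic material parameters over particle life.
# GPU Sparkles spray from a flattened, expanding half-sphere arc, shaped by gravity, drag, inherited velocity and vector noise.
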NiagaraSystem /BRCosmetics/Gadgets/Assets/VinderTech_GliderChute/Glider_Marshmellow/FX/NS_Glider_Mello_Deploy_Gold
dependencies {usedDataInterfaces [&NiagaraDataInterfaceCurve], usedDynamicInputs [], usedModules [&SystemState]}
emitters {
 Center {
  emitterUpdateScript [
   &EmitterState(Life Cycle Mode=::NewEnumerator1, Loop Behavior=::NewEnumerator1, Loop Duration=0.5, UseLoopDelay=false)
   &SpawnRate(SpawnRate=10, Use Spawn Probability=false)
  ]
  particleSpawnScript [
   &InitializeParticle(
    Color=<rgba 0.2135 0.159 0 1>
    Lifetime Max=0.1
    Lifetime Min=0.05
    Lifetime Mode=::NewEnumerator1
    Sprite Size Mode=::NewEnumerator3
    Uniform Sprite Size=50
    UsePositionOffset=false
    Write Material Random=true
   )
   &CameraOffset(Camera Offset Amount=400)
   &DynamicMaterialParameters(
    Index 0 Param 1=1
    Index 0 Param 2=1
    Index 0 Param 3=1
    Index 0 Param 4=1
    Module.Param0WriteEnabled=true
    Module.Param1WriteEnabled=true
    Module.Param2WriteEnabled=true
    Module.Param3WriteEnabled=true
   )
  ]
  particleUpdateScript [&ParticleState()]
  rendererClasses [&NiagaraSpriteRendererProperties]
  renderers [{rendererClass=&NiagaraSpriteRendererProperties}]
 }
 Mello {
  emitterUpdateScript [
   &EmitterState()
   &SpawnBurst_Instantaneous(Use Spawn Probability=false, UseLoopCountLimit=false)
   /SetVariables(Emitter.RandomSpeed:Vector4f=&MakeVector4(
     X=&RandomRangeFloat(Minimum=-1, Maximum=1)
     Y=&RandomRangeFloat(Minimum=-1, Maximum=1)
     Z=0
     W=0
    )) {isSetParameters=true}
  ]
  particleSpawnScript [&InitializeParticle(Color Mode=::NewEnumerator0, Sprite Size Mode=::NewEnumerator3, Uniform Sprite Size=30, UsePositionOffset=false, Write Material Random=true)]
  particleUpdateScript [
   &ParticleState()
   &CameraOffset(Camera Offset Amount=&FloatFromCurve(FloatCurve=DataInterface() {propertyValues {Curve {keys [<at 0 100>,<at 1 200>]}}}))
   &ScaleSpriteSize(Uniform Curve Sprite Scale=DataInterface() {propertyValues {Curve {keys [{tangentMode=RCTM_User, time=0, value=0}, {tangentMode=RCTM_User, time=0.25, value=10}]}}})
   &ScaleColor(
    Scale Alpha=&FloatFromCurve(FloatCurve=DataInterface() {propertyValues {Curve {keys [<at 0.75 1>,<at 1 0>]}}})
    Scale RGB=&VectorFromFloat(Value=&FloatFromCurve(FloatCurve=DataInterface() {propertyValues {Curve {keys [<at 0 10>,<at 0.25 1>]}}}))
    ScaleA=true
    ScaleRGB=true
   )
   &DynamicMaterialParameters(
    Index 0 Param 1=&FloatFromCurve(FloatCurve=DataInterface() {propertyValues {Curve {keys [<at 0 1>,<at 0.25 -1>]}}})
    Index 0 Param 2=&FloatFromCurve(FloatCurve=DataInterface() {propertyValues {Curve {keys [<at 0.25 1>,<at 1 -1>]}}})
    Index 0 Param 3=&FloatFromCurve(FloatCurve=DataInterface() {propertyValues {Curve {keys [
         {interpMode=RCIM_Cubic, leaveTangentWeight=0.4998, tangentMode=RCTM_User, time=0, value=0}
         {arriveTangentWeight=0.1274, interpMode=RCIM_Cubic, leaveTangentWeight=0.7778, tangentMode=RCTM_User, time=0.25, value=0.9847}
         {arriveTangent=-0.0405, interpMode=RCIM_Cubic, leaveTangent=-0.0405, tangentMode=RCTM_User, time=1, value=0}
        ]}}})
    Index 0 Param 4=&FloatFromCurve(FloatCurve=DataInterface() {propertyValues {Curve {keys [
         {interpMode=RCIM_Cubic, leaveTangentWeight=1.7224, tangentMode=RCTM_User, time=0, value=0}
         {arriveTangentWeight=0.1274, interpMode=RCIM_Cubic, leaveTangentWeight=0.7778, tangentMode=RCTM_User, time=0.25, value=0.9847}
         {arriveTangent=-0.0405, interpMode=RCIM_Cubic, leaveTangent=-0.0405, tangentMode=RCTM_User, time=1, value=0}
        ]}}})
    Module.Param0WriteEnabled=true
    Module.Param1WriteEnabled=true
    Module.Param2WriteEnabled=true
    Module.Param3WriteEnabled=true
   )
  ]
  rendererClasses [&NiagaraSpriteRendererProperties]
  renderers [{rendererClass=&NiagaraSpriteRendererProperties}]
 }
 Mesh_Sprites {
  emitterUpdateScript [
   &EmitterState()
   &SpawnBurst_Instantaneous(Use Spawn Probability=false, UseLoopCountLimit=false)
  ]
  particleSpawnScript [
   &InitializeParticle(
    Color=<rgba 2 1.4888 0 0>
    Lifetime=2
    Mesh Scale Mode=::NewEnumerator3
    Mesh Uniform Scale=0.009
    UsePositionOffset=false
    Write Material Random=true
   )
   &CameraOffset(Camera Offset Amount=15)
  ]
  particleUpdateScript [
   &ParticleState()
   &ScaleMeshSize(Scale Factor=&VectorFromFloat(Value=&FloatFromCurve(FloatCurve=DataInterface() {propertyValues {Curve {keys [
          {arriveTangent=20.3465, interpMode=RCIM_Cubic, leaveTangent=20.3465, leaveTangentWeight=0.3715, tangentMode=RCTM_User, time=0, value=0}
          {
           arriveTangent=0.1372
           arriveTangentWeight=0.0508
           interpMode=RCIM_Cubic
           leaveTangent=0.1372
           leaveTangentWeight=0.4572
           tangentMode=RCTM_User
           time=0.1
           value=1.11
          }
          {arriveTangentWeight=0.3003, interpMode=RCIM_Cubic, tangentMode=RCTM_User, time=1, value=1.15}
         ]}}})))
   &DynamicMaterialParameters(
    Index 0 Param 1=&FloatFromCurve(FloatCurve=DataInterface() {propertyValues {Curve {keys [
         {arriveTangent=26.7638, interpMode=RCIM_Cubic, leaveTangent=26.7638, leaveTangentWeight=1.2171, tangentMode=RCTM_User, time=0, value=0}
         {
          arriveTangent=67.7974
          arriveTangentWeight=2.1669
          interpMode=RCIM_Cubic
          leaveTangent=67.7974
          leaveTangentWeight=2.1669
          tangentMode=RCTM_User
          time=0.1
          value=3.65
         }
         {arriveTangentWeight=0.1037, interpMode=RCIM_Cubic, leaveTangentWeight=0.8292, tangentMode=RCTM_User, time=0.2, value=13}
         {arriveTangentWeight=4.0089, interpMode=RCIM_Cubic, tangentMode=RCTM_User, time=1, value=1}
        ]}}})
    Index 0 Param 2=&FloatFromCurve(FloatCurve=DataInterface() {propertyValues {Curve {keys [
         {arriveTangent=131.7274, interpMode=RCIM_Cubic, leaveTangent=131.7274, leaveTangentWeight=18.6671, tangentMode=RCTM_User, time=0, value=0}
         {
          arriveTangent=131.7274
          arriveTangentWeight=10.4009
          interpMode=RCIM_Cubic
          leaveTangent=131.7274
          leaveTangentWeight=15.6013
          tangentMode=RCTM_User
          time=0.4
          value=56
         }
         {arriveTangentWeight=7.3361, interpMode=RCIM_Cubic, tangentMode=RCTM_User, time=1, value=78}
        ]}}})
    Index 0 Param 3=1
    Module.Param0WriteEnabled=true
    Module.Param1WriteEnabled=true
    Module.Param2WriteEnabled=true
    Module.Param3WriteEnabled=true
   )
   &ScaleColor(
    Scale RGB=&VectorFromFloat(Value=&FloatFromCurve(FloatCurve=DataInterface() {propertyValues {Curve {keys [
          {interpMode=RCIM_Cubic, leaveTangentWeight=12.8652, tangentMode=RCTM_User, time=0, value=7}
          {arriveTangentWeight=0.0833, interpMode=RCIM_Cubic, tangentMode=RCTM_User, time=0.25, value=1}
         ]}}}))
    ScaleA=true
    ScaleRGB=true
   )
   &GravityForce(Gravity=<0 0 -100>)
   &SolveForcesAndVelocity(Clamp Velocity=false, Limit Acceleration=false)
  ]
  rendererClasses [&NiagaraMeshRendererProperties]
  renderers [{rendererClass=&NiagaraMeshRendererProperties, rendererIndex=0}]
 }
 Sparkles {
  simTarget=GPUComputeSim
  emitterUpdateScript [
   &EmitterState(Life Cycle Mode=::NewEnumerator1, Loop Behavior=::NewEnumerator1, Loop Duration=0.3, UseLoopDelay=false)
   &SpawnRate(SpawnRate=1250, Use Spawn Probability=false)
  ]
  particleSpawnScript [
   &InitializeParticle(
    Color=<rgba 50 50 50 1>
    Lifetime Max=1.5
    Lifetime Min=0.5
    Lifetime Mode=::NewEnumerator1
    Sprite Rotation Mode=::NewEnumerator2
    Sprite Size Mode=::NewEnumerator1
    Uniform Sprite Size Max=8
    UsePositionOffset=false
    Write Material Random=true
   )
   &SphereLocation(
    Hemisphere X=true
    Hemisphere Z=true
    MaskSpawnGroup=false
    Non Uniform Scale=<0 1 1>
    Offset=<0 0 15>
    Override Seed=false
    Sphere Radius=&FloatFromCurve(FloatCurve=DataInterface() {propertyValues {Curve {keys [<at 0 15>,<at 1 100>]}}}, CurveIndex=$Emitter.NormalizedLoopAge)
    Surface Expansion Mode=::NewEnumerator1
    Surface Only=true
    UseNonUniformScale=true
    UseOffset=true
    UseSphereOrientation=false
   )
   &SphereLocation(
    MaskSpawnGroup=false
    Override Seed=false
    Sphere Radius=5
    Surface Only=false
    UseNonUniformScale=false
    UseOffset=false
    UseSphereOrientation=false
   ) {module=SphereLocation001}
   &AddVelocity(Velocity=&Multiply_VectorByFloat(Vector=$Output.SphereLocation.SphereNormal, Float=&RandomRangeFloat(Minimum=25, Maximum=75)))
   &AddVelocity(Velocity=&RandomRangeVector(Minimum=<-80 -254 100>, Maximum=<187 420 150>)) {module=AddVelocity001}
   &AddVelocity(Velocity=<0 0 -250>) {module=AddVelocity002}
   &AddVelocityInCone(
    Cone Angle=0
    Cone Axis=<-1 0 0>
    Module.Use Velocity Falloff On Cone Axis=false
    Velocity Strength=&RandomRangeFloat(Minimum=20, Maximum=250)
   )
   &DynamicMaterialParameters(
    Index 0 Param 1=&UniformAOrBFloat(A=0, Distribution Weight=0.25)
    Module.Param0WriteEnabled=true
    Module.Param1WriteEnabled=true
    Module.Param2WriteEnabled=true
    Module.Param3WriteEnabled=true
   )
  ]
  particleUpdateScript [
   &GravityForce(Gravity=<0 0 -250>)
   &Drag(Drag=2, Use Linear Drag=true, Use Rotational Drag=false)
   &InheritVelocity(Inherited Velocity Amount Scale=<0.6 0.6 0.6>, Speed Limit=true, UseSpeedThreshold=false)
   &VectorNoiseForce(Force Amount=2500, Module.Use Random Noise Time Interval=false)
   &SolveForcesAndVelocity(Clamp Velocity=false, Limit Acceleration=false)
   &ParticleState()
  ]
  rendererClasses [&NiagaraSpriteRendererProperties]
  renderers [{rendererClass=&NiagaraSpriteRendererProperties}]
 }
}
summary {
 systemName=NS_Glider_Mello_Deploy_Gold
 emitters [
  {emitterName=Mesh_Sprites, rendererClasses [&NiagaraMeshRendererProperties]}
  {emitterName=Sparkles, simTarget=GPUComputeSim, rendererClasses [&NiagaraSpriteRendererProperties]}
  {emitterName=Center, rendererClasses [&NiagaraSpriteRendererProperties]}
  {emitterName=Mello, rendererClasses [&NiagaraSpriteRendererProperties]}
 ]
 userVariables []
}
