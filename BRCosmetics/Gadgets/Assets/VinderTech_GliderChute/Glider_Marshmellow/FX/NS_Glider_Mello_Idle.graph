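# Niagara idle effect for the Marshmellow glider, with ribbons, edge glows, flares and small bit particles attached to its skeletal mesh.
# Paired spawn groups select two sockets or bones on User.GliderMesh, giving each layer two emission points.
# A system timeline ramps glow and bit spawning, including movement-based GPU Digi-Bits, and threshold-gates ribbons.
# User.bisFullyDeployed gates flares and a looping component-renderer particle driving light intensity at a third bone.
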NiagaraSystem /BRCosmetics/Gadgets/Assets/VinderTech_GliderChute/Glider_Marshmellow/FX/NS_Glider_Mello_Idle
dependencies {
 usedDataInterfaces [&NiagaraDataInterfaceVectorCurve, &NiagaraDataInterfaceCurve]
 usedDynamicInputs [
  &Boolean_AndOperation
  &Boolean_OrOperation
  &Divide_Float
 ]
 usedModules [
  &SystemState
  &Timeline
  &self:SystemSpawnScript.SystemScriptSource.SystemScriptGraph.NiagaraNodeAssignment_0.SetVariables_#1
 ]
}
emitters {
 Digi-Bits {
  simTarget=GPUComputeSim
  emitterUpdateScript [
   &EmitterState()
   &SpawnPerUnit(
    Module.Use Max Movement Threshold=true
    Module.Use Movement Tolerance=false
    Spawn Probability:NiagaraFloat=&Lerp_Float(A=0, B=1, Alpha=$System.Timeline.NormalizedTime)
    Spawn Spacing=30
    Use Spawn Probability=true
   )
   &SpawnPerUnit(
    Module.Use Max Movement Threshold=true
    Module.Use Movement Tolerance=false
    Spawn Group=1
    Spawn Probability:NiagaraFloat=&Lerp_Float(A=0, B=1, Alpha=$System.Timeline.NormalizedTime)
    Spawn Spacing=30
    Use Spawn Probability=true
   ) {module=SpawnPerUnit001}
  ]
  particleSpawnScript [
   &InitializeParticle(
    Color Mode=::NewEnumerator0
    Lifetime Max=0.25
    Lifetime Min=0.15
    Lifetime Mode=::NewEnumerator1
    Sprite Size Mode=::NewEnumerator1
    Uniform Sprite Size Max=7
    Uniform Sprite Size Min=2
    UsePositionOffset=false
    Write Material Random=true
   )
   &SkeletalMeshLocation(
    Bone / Socket Index=&SpawnGroup()
    MaskSpawnGroup=false
    Mesh Sampling Type=::NewEnumerator5
    Skeletal Mesh=$User.GliderMesh
    Socket Sampling Mode=::NewEnumerator5
   )
   &SphereLocation(
    MaskSpawnGroup=false
    Override Seed=false
    Sphere Radius=5
    Surface Only=false
    UseNonUniformScale=false
    UseOffset=false
    UseSphereOrientation=false
   )
   &AddVelocity(
    Coordinate Space=::Simulation
    Velocity=&Multiply_VectorByFloat(Vector=$Output.SphereLocation.SphereNormal, Float=&RandomRangeFloat(Minimum=20, Maximum=50))
   )
  ]
  particleUpdateScript [
   &InheritVelocity(Speed Limit=true, UseSpeedThreshold=false)
   &Drag(Drag=&RandomRangeFloat(Minimum=1, Maximum=5), Use Linear Drag=true, Use Rotational Drag=false)
   &SolveForcesAndVelocity(Clamp Velocity=false, Limit Acceleration=false)
   &ParticleState()
   &ScaleColor(
    Scale Alpha=&FloatFromCurve(FloatCurve=DataInterface() {propertyValues {Curve {keys [
         {interpMode=RCIM_Cubic, tangentMode=RCTM_User, time=0, value=0}
         {interpMode=RCIM_Cubic, tangentMode=RCTM_User, time=0.5, value=1}
         {tangentMode=RCTM_User, time=1, value=0}
        ]}}})
    Scale RGB=&VectorFromFloat(Value=100)
    ScaleA=true
    ScaleRGB=true
   )
   &ScaleSpriteSize(Uniform Curve Sprite Scale=DataInterface() {propertyValues {Curve {keys [<at 0 1>,<at 0.7 0.25>,<at 1 0>]}}})
  ]
  rendererClasses [&NiagaraSpriteRendererProperties]
  renderers [{rendererClass=&NiagaraSpriteRendererProperties}]
 }
 Edge_Glows {
  emitterUpdateScript [
   &EmitterState()
   &SpawnRate(SpawnRate=&Multiply_Float(A=30, B=$System.Timeline.NormalizedTime), Use Spawn Probability=false)
   &SpawnRate(SpawnGroup=1, SpawnRate=&Multiply_Float(A=30, B=$System.Timeline.NormalizedTime), Use Spawn Probability=false) {module=SpawnRate001}
  ]
  particleSpawnScript [
   &InitializeParticle(
    Color=<rgba 1 1 1 0.25>
    Lifetime Max=0.4
    Lifetime Min=0.2
    Lifetime Mode=::NewEnumerator1
    Sprite Size Mode=::NewEnumerator1
    Uniform Sprite Size Max=45
    Uniform Sprite Size Min=30
    UsePositionOffset=false
    Write Material Random=true
   )
   &SkeletalMeshLocation(
    Bone / Socket Index=&SpawnGroup()
    MaskSpawnGroup=false
    Mesh Sampling Type=::NewEnumerator5
    Skeletal Mesh=$User.GliderMesh
    Socket Sampling Mode=::NewEnumerator5
   )
   &AddVelocity(Coordinate Space=::World, Velocity=&RandomRangeVector(Minimum=<30 -10 -10>, Maximum=<30 10 10>))
   &DynamicMaterialParameters(
    Index 0 Param 1=&RandomRangeFloat(Minimum=2.5, Maximum=3)
    Index 0 Param 2=1
    Index 0 Param 3=1
    Index 0 Param 4=1
    Module.Param0WriteEnabled=true
    Module.Param1WriteEnabled=true
    Module.Param2WriteEnabled=true
    Module.Param3WriteEnabled=true
   )
   &CameraOffset(Camera Offset Amount=-20)
  ]
  particleUpdateScript [
   &ParticleState()
   &AccelerationForce(Acceleration=<0 0 2>)
   &SolveForcesAndVelocity(Clamp Velocity=false, Limit Acceleration=false)
   &ScaleColor(
    Scale Alpha=&FloatFromCurve(FloatCurve=DataInterface() {propertyValues {Curve {keys [
         {arriveTangent=3.3588, interpMode=RCIM_Cubic, leaveTangent=3.3588, tangentMode=RCTM_User, time=0, value=0}
         {arriveTangentWeight=0.3621, interpMode=RCIM_Cubic, leaveTangentWeight=0.5431, tangentMode=RCTM_User, time=0.4, value=0.9847}
         {arriveTangent=-0.0405, interpMode=RCIM_Cubic, leaveTangent=-0.0405, tangentMode=RCTM_User, time=1, value=0}
        ]}}})
    ScaleA=true
    ScaleRGB=true
   )
   &ScaleSpriteSize(Uniform Curve Sprite Scale=DataInterface() {propertyValues {Curve {keys [
        {interpMode=RCIM_Cubic, leaveTangentWeight=0.2867, tangentMode=RCTM_User, time=0, value=0.5}
        {arriveTangentWeight=0.238, interpMode=RCIM_Cubic, leaveTangentWeight=0.102, tangentMode=RCTM_User, time=0.7, value=1}
        {arriveTangentWeight=0.1414, time=1, value=0.7}
       ]}}})
  ]
  rendererClasses [&NiagaraSpriteRendererProperties]
  renderers [{rendererClass=&NiagaraSpriteRendererProperties}]
 }
 Empty {
  emitterUpdateScript [
   &EmitterState()
   &SpawnRate(
    SpawnRate=&Multiply_Float(A=75, B=&Lerp_Float(A=0, B=1, Alpha=$System.Timeline.NormalizedTime))
    Use Spawn Probability=false
   )
   &SpawnRate(
    SpawnGroup=1
    SpawnRate=&Multiply_Float(A=75, B=&Lerp_Float(A=0, B=1, Alpha=$System.Timeline.NormalizedTime))
    Use Spawn Probability=false
   ) {module=SpawnRate001}
  ]
  particleSpawnScript [
   &InitializeParticle(
    Color=<rgba 500 500 500 1>
    Lifetime Max=0.75
    Lifetime Min=0.25
    Lifetime Mode=::NewEnumerator1
    Sprite Rotation Mode=::NewEnumerator2
    Sprite Size Mode=::NewEnumerator1
    Uniform Sprite Size Max=5
    Uniform Sprite Size Min=3
    UsePositionOffset=false
    Write Material Random=true
   )
   &SkeletalMeshLocation(
    Bone / Socket Index=&SpawnGroup()
    Bone Sampling Mode=::NewEnumerator4
    MaskSpawnGroup=false
    Skeletal Mesh=$User.GliderMesh
   )
   &BoxLocation(
    Box Origin:NiagaraPosition=$Particles.SkeletalMeshLocation.SampledPosition
    Box Size=<0 80 5>
    MaskSpawnGroup=false
    Module.Override Seed=false
    Rotate Box=false
    Surface Only=false
   )
   &AddVelocity(Velocity=&MakeVector(
     X=&MaskFloatBySpawnGroup(Masked Float=-200, Passthrough Float=200)
     Y=0
     Z=&RandomRangeFloat(Minimum=-50, Maximum=-75)
    ))
  ]
  particleUpdateScript [
   &ParticleState()
   &GravityForce(Gravity=<0 0 -250>)
   &Drag(Drag=2, Use Linear Drag=true, Use Rotational Drag=false)
   &InheritVelocity(Inherited Velocity Amount Scale=<-0.1 -0.1 0>, Speed Limit=true, UseSpeedThreshold=false)
   &VectorNoiseForce(Force Amount=750, Module.Use Random Noise Time Interval=false)
   &SolveForcesAndVelocity(Clamp Velocity=false, Limit Acceleration=false)
   &ScaleSpriteSize(Uniform Curve Sprite Scale=DataInterface() {propertyValues {Curve {keys [{tangentMode=RCTM_User, time=0, value=1}, {tangentMode=RCTM_User, time=1, value=0}]}}})
  ]
  rendererClasses [&NiagaraSpriteRendererProperties]
  renderers [{rendererClass=&NiagaraSpriteRendererProperties}]
 }
 Empty001 {
  emitterUpdateScript [
   &EmitterState(Life Cycle Mode=::NewEnumerator1, Loop Duration=$System.LightTimelinePlayrate, UseLoopDelay=false)
   &SpawnBurst_Instantaneous(
    Spawn Count=&Multiply_Int(A=1, B=&MakeCustomIntFromBool(Bool=$User.bisFullyDeployed, True INT=1, False INT=0))
    Use Spawn Probability=false
    UseLoopCountLimit=false
   )
  ]
  particleSpawnScript [&InitializeParticle(
    Color Channel Mode:ENiagaraChannelCorrelation=::NewEnumerator2
    Color Mode=::NewEnumerator2
    Lifetime=$System.LightTimelinePlayrate
    UsePositionOffset=false
    Write Material Random=true
   )]
  particleUpdateScript [
   &ParticleState()
   &SkeletalMeshLocation(
    Bone / Socket Index=2
    Bone Sampling Mode=::NewEnumerator4
    MaskSpawnGroup=false
    Sampled Position Offset=<0 0 20>
    Skeletal Mesh=$User.GliderMesh
   )
   /SetVariables(Particles.LightIntensity:NiagaraFloat=&FloatFromCurve(FloatCurve=DataInterface() {propertyValues {Curve {keys [<at 0 0>,<at 1 25>]}}})) {isSetParameters=true}
  ]
  rendererClasses [&NiagaraComponentRendererProperties]
  renderers [{rendererClass=&NiagaraComponentRendererProperties, rendererIndex=0}]
 }
 Flares {
  emitterUpdateScript [
   &EmitterState()
   &SpawnRate(
    SpawnRate=&MakeCustomFloatFromBool(Bool=$User.bisFullyDeployed, True Float=30, False Float=0)
    Use Spawn Probability=false
   )
   &SpawnRate(
    SpawnGroup=1
    SpawnRate=&MakeCustomFloatFromBool(Bool=$User.bisFullyDeployed, True Float=30, False Float=0)
    Use Spawn Probability=false
   ) {module=SpawnRate001}
  ]
  particleSpawnScript [
   &InitializeParticle(
    Color Mode=::NewEnumerator0
    Lifetime Max=0.4
    Lifetime Min=0.3
    Lifetime Mode=::NewEnumerator1
    Sprite Size Mode=::NewEnumerator3
    Uniform Sprite Size=60
    UsePositionOffset=false
    Write Material Random=true
   )
   &DynamicMaterialParameters(
    Index 0 Param 1=0.6
    Index 0 Param 2=1
    Index 0 Param 3=1
    Index 0 Param 4=1
    Module.Param0WriteEnabled=true
    Module.Param1WriteEnabled=true
    Module.Param2WriteEnabled=true
    Module.Param3WriteEnabled=true
   )
   &CameraOffset(Camera Offset Amount=50)
  ]
  particleUpdateScript [
   &ParticleState()
   &SkeletalMeshLocation(
    Bone / Socket Index=&SpawnGroup()
    MaskSpawnGroup=false
    Mesh Sampling Type=::NewEnumerator5
    Skeletal Mesh=$User.GliderMesh
    Socket Sampling Mode=::NewEnumerator5
   )
   &ScaleColor(
    Scale Alpha=&FloatFromCurve(FloatCurve=DataInterface() {propertyValues {Curve {keys [
         {arriveTangent=3.3588, interpMode=RCIM_Cubic, leaveTangent=3.3588, tangentMode=RCTM_User, time=0, value=0}
         {arriveTangentWeight=0.3621, interpMode=RCIM_Cubic, leaveTangentWeight=0.5431, tangentMode=RCTM_User, time=0.4, value=0.9847}
         {arriveTangent=-0.0405, interpMode=RCIM_Cubic, leaveTangent=-0.0405, tangentMode=RCTM_User, time=1, value=0}
        ]}}})
    Scale RGB=&VectorFromFloat(Value=&FloatFromCurve(FloatCurve=DataInterface() {propertyValues {Curve {keys [<at 0 2>,<at 1 1>]}}}))
    ScaleA=true
    ScaleRGB=true
   )
   &ScaleSpriteSize(Uniform Curve Sprite Scale=DataInterface() {propertyValues {Curve {keys [
        {interpMode=RCIM_Cubic, leaveTangentWeight=0.2867, tangentMode=RCTM_User, time=0, value=0.5}
        {arriveTangentWeight=0.238, interpMode=RCIM_Cubic, leaveTangentWeight=0.102, tangentMode=RCTM_User, time=0.7, value=1}
        {arriveTangentWeight=0.1414, time=1, value=0.7}
       ]}}})
  ]
  rendererClasses [&NiagaraSpriteRendererProperties]
  renderers [{rendererClass=&NiagaraSpriteRendererProperties}]
 }
 Ribbons {
  emitterUpdateScript [
   &EmitterState()
   &SpawnRate(
    SpawnRate=&MakeCustomFloatFromBool(
     Bool=&SetBoolByFloatComparison(A=$System.Timeline.NormalizedTime, B=0.5, Comparison Type=::NewEnumerator2)
     True Float=120
     False Float=0
    )
    Use Spawn Probability=false
   )
   &SpawnRate(
    SpawnGroup=1
    SpawnRate=&MakeCustomFloatFromBool(
     Bool=&SetBoolByFloatComparison(A=$System.Timeline.NormalizedTime, B=0.5, Comparison Type=::NewEnumerator2)
     True Float=120
     False Float=0
    )
    Use Spawn Probability=false
   ) {module=SpawnRate001}
  ]
  particleSpawnScript [
   &InitializeRibbon(Color=<rgba 25 25 25 1>, Lifetime=0.15, Ribbon Width=8, UsePositionOffset=false, Write Material Random=true)
   &AddVelocity(Velocity=<0 0 500>) {enabled=false}
   &SkeletalMeshLocation(
    Bone / Socket Index=&SpawnGroup()
    Interpolate Mesh Sampling=false
    MaskSpawnGroup=false
    Mesh Sampling Type=::NewEnumerator5
    Position Sampling=::NewEnumerator1
    Skeletal Mesh=$User.GliderMesh
    Socket Sampling Mode=::NewEnumerator5
   )
  ]
  particleUpdateScript [
   &SolveForcesAndVelocity(Clamp Velocity=false, Limit Acceleration=false) {enabled=false}
   &ParticleState()
   &ScaleColor(
    Scale Alpha=&FloatFromCurve(FloatCurve=DataInterface() {propertyValues {Curve {keys [<at 0 1>,<at 1 0>]}}})
    Scale RGB=&VectorFromCurve(VectorCurve=DataInterface() {propertyValues {
       XCurve {keys [
         {interpMode=RCIM_Cubic, leaveTangentWeight=3.4854, tangentMode=RCTM_User, time=0, value=-10.1885}
         {arriveTangentWeight=0.9768, interpMode=RCIM_Cubic, leaveTangentWeight=2.442, tangentMode=RCTM_User, time=0.1, value=0.2674}
         {arriveTangentWeight=0.1, interpMode=RCIM_Cubic, leaveTangentWeight=0.06, tangentMode=RCTM_User, time=0.35, value=0.062}
         {arriveTangentWeight=0.0538, interpMode=RCIM_Cubic, tangentMode=RCTM_User, time=0.5, value=0.0025}
        ]}
       YCurve {keys [
         {interpMode=RCIM_Cubic, leaveTangentWeight=0.5172, tangentMode=RCTM_User, time=0, value=-1.5014}
         {arriveTangentWeight=0.1482, interpMode=RCIM_Cubic, leaveTangentWeight=0.3704, tangentMode=RCTM_User, time=0.1, value=0.047}
         {arriveTangentWeight=0.0838, interpMode=RCIM_Cubic, leaveTangentWeight=0.0503, tangentMode=RCTM_User, time=0.35, value=0.0145}
         {arriveTangentWeight=0.0502, interpMode=RCIM_Cubic, tangentMode=RCTM_User, time=0.5, value=0.0026}
        ]}
       ZCurve {keys [
         {interpMode=RCIM_Cubic, leaveTangentWeight=2.78, tangentMode=RCTM_User, time=0, value=-7.9393}
         {arriveTangentWeight=0.7711, interpMode=RCIM_Cubic, leaveTangentWeight=1.9278, tangentMode=RCTM_User, time=0.1, value=0.4}
         {arriveTangentWeight=0.1141, interpMode=RCIM_Cubic, leaveTangentWeight=0.0684, tangentMode=RCTM_User, time=0.35, value=0.15}
         {arriveTangentWeight=0.0649, interpMode=RCIM_Cubic, tangentMode=RCTM_User, time=0.5, value=0.026}
        ]}
      }})
    ScaleA=true
    ScaleRGB=true
   )
   &ScaleRibbonWidth(Ribbon Width Scale=&FloatFromCurve(FloatCurve=DataInterface() {propertyValues {Curve {keys [{tangentMode=RCTM_User, time=0, value=1}, {tangentMode=RCTM_User, time=1, value=2}]}}}))
  ]
  rendererClasses [&NiagaraRibbonRendererProperties]
  renderers [{rendererClass=&NiagaraRibbonRendererProperties}]
 }
}
summary {
 systemName=NS_Glider_Mello_Idle
 emitters [
  {emitterName=Ribbons, rendererClasses [&NiagaraRibbonRendererProperties]}
  {emitterName=Edge_Glows, rendererClasses [&NiagaraSpriteRendererProperties]}
  {emitterName=Digi-Bits, simTarget=GPUComputeSim, rendererClasses [&NiagaraSpriteRendererProperties]}
  {emitterName=Flares, rendererClasses [&NiagaraSpriteRendererProperties]}
  {emitterName=Empty, rendererClasses [&NiagaraSpriteRendererProperties]}
  {emitterName=Empty001, rendererClasses [&NiagaraComponentRendererProperties]}
 ]
 userVariables [
  User.bIsFrontEndPreview {type=NiagaraBool}
  User.bisFullyDeployed {type=NiagaraBool}
  User.ForwardDot {type=NiagaraFloat, defaultValue=0, description=""}
  User.GliderMesh {
   type=NiagaraDataInterfaceSkeletalMesh
   defaultValue {
    struct=NiagaraExt_VariableValue_DataInterface
    value {dataInterface=&self:NiagaraDataInterfaceSkeletalMesh_0}
   }
  }
 ]
}
